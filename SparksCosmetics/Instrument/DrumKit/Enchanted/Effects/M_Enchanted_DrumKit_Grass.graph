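# Unlit, translucent, two-sided particle material for grass effects on the Enchanted drum kit cosmetic, usable on Niagara sprites and instanced meshes.
# Particle color is tinted for time of day, with a static switch choosing a curve atlas row driven by a parameter collection or four color parameters.
# A dynamic parameter shifts the hue before a shared emissive adjustment function.
# A grass texture's green channel is multiplied by particle alpha.
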Material /SparksCosmetics/Instrument/DrumKit/Enchanted/Effects/M_Enchanted_DrumKit_Grass
graph {
 properties {
  bUsedWithInstancedStaticMeshes=true
  bUsedWithNiagaraSprites=true
  blendMode=BLEND_Translucent
  opacityMaskClipValue=0.1
  pixelDepthOffsetMode=PDOM_Legacy
  shadingModel=MSM_Unlit
  twoSided=true
  usedShadingModels=Unlit
 }
 nodes [
  n1 = ParticleColor()
  n2 = CurveAtlasRowParameter(CurveTime=CollectionParameter() {collection=&FortniteMaterialParameters, parameterName=TimeOfDay_0-1}) {
   atlas=&AthenaCurveAtlas01
   curve=/Game/Athena/LinearColorCurves/Curve_TimeOfDay_S23
   defaultValue=321
   group=TimeOfDay
   parameterName=TOD_Curve
  }
  n3 = MaterialFunctionCall(
   Color_Morning=VectorParameter() {parameterName="Color_Morning (V3)", defaultValue=<rgba 1 1 1 0>}.RGB
   Color_Day=VectorParameter() {parameterName="Color_Day (V3)", defaultValue=<rgba 1 1 1 0>}.RGB
   Color_Evening=VectorParameter() {parameterName="Color_Evening (V3)", defaultValue=<rgba 1 1 1 0>}.RGB
   Color_Night=VectorParameter() {parameterName="Color_Night (V3)", defaultValue=<rgba 1 1 1 0>}.RGB
  ) {materialFunction=&Color_Time_of_Day}
  n4 = MaterialFunctionCall(
   Hue Shift Percentage=DynamicParameter() {paramNames [HueShift, EmissiveAmount, NearCamFadeDist, NearCamScaleDist]}.HueShift
   Texture=StaticSwitchParameter(True=Multiply(A=@n1.RGB, B=@n2), False=Multiply(A=@n1.RGB, B=@n3.Result)) {parameterName=Use ToD Color Curve}
  ) {materialFunction=&HueShift}
  MaterialFunctionCall(Emissive=@n4.Result) {materialFunction=&MF_FX_EmissiveAdjustments}
  Multiply(
   A=ParticleColor().A
   B=TextureSample() {samplerType=SAMPLERTYPE_Grayscale, texture=/Game/Effects/Fort_Effects/Textures/Environments/Apollo/T_Apollo_Grass01}.G
  )
  Reroute(None=@n4.Result)
 ]
}
