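# Translucent unlit smoke material for Niagara sprite effects on the Enchanted DrumKit instrument cosmetic.
# A 4x4 flipbook uses motion vectors to blend frames, with its phase driven by the Animation dynamic parameter.
# The flipbook's green channel multiplies particle color into an emissive adjustment function and particle alpha into opacity.
# Depth fade softens the opacity where sprites meet scene geometry.
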Material /SparksCosmetics/Instrument/DrumKit/Enchanted/Effects/M_Enchanted_DrumKit_Smoke
graph {
 properties {
  bCastRayTracedShadows=false
  bUsedWithNanite=true
  bUsedWithNiagaraSprites=true
  blendMode=BLEND_Translucent
  pixelDepthOffsetMode=PDOM_Legacy
  shadingModel=MSM_Unlit
  translucencyLightingMode=TLM_VolumetricPerVertexDirectional
  translucentSelfShadowDensityScale=1
  usedShadingModels=Unlit
  translucentMultipleScatteringExtinction=<rgba 1 1 1 1>
 }
 nodes [
  n1 = NamedRerouteUsage()
  n2 = ParticleColor()
  DepthFade(Opacity=Multiply(A=@n1, B=@n2.A), FadeDistance=ScalarParameter() {defaultValue=3})
  MaterialFunctionCall(Emissive=Multiply(A=@n2.RGB, B=@n1)) {materialFunction=&MF_FX_EmissiveAdjustments}
  n3 = DynamicParameter() {paramNames [Animation, Unused, Emissive, NormalMultiplier]}
  n4 = Constant() {r=4}
  n5 = TextureObject() {texture=/Game/Effects/Niagara/MasterMaterials/Smoke/Textures/NEW/RT4x4_4_Tex}
  n6 = TextureObjectParameter() {parameterName=MotionVector, samplerType=SAMPLERTYPE_LinearColor, texture=/Game/Effects/Niagara/MasterMaterials/Smoke/Textures/Vector_OF}
  n7 = ScalarParameter() {defaultValue=0.344, group=Motion Vectors, parameterName=MotionVector}
  NamedRerouteDeclaration(None=ComponentMask(None=MaterialFunctionCall!(
     "Animation  Phase (0-1)"=@n3.Animation
     Number of Columns X=@n4
     Number of Rows Y=@n4
     Texture=@n5
     UVs
     Motion Vector Texture=@n6
     Phase to Debug
     Motion Vector Intensity=@n7
     Double Apply Motion Vectors
     Custom UV Input
     Custom UV2 Input
     Use Custom UVs
    ) {materialFunction=/Engine/Functions/Engine_MaterialFunctions02/Texturing/FlipBook_MotionVectors}.Result) {g=true}) {name=Alpha_Flipbook, nodeColor=<rgba 1 0 0.3516 1>}
 ]
}
